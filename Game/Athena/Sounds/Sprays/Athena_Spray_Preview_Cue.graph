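# Athena sound cue for previewing sprays, playing a spray paint sequence sound.
# A random node picks one of three equally weighted preview wave variations.
# It uses the emote music sound class, emote attenuation and a reduced volume.
# A concurrency override allows a single instance at a time, stopping the oldest when a new one starts.
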SoundCue /Game/Athena/Sounds/Sprays/Athena_Spray_Preview_Cue
root=Random_0
nodes [
 {id=Random_0, children [WavePlayer_1,WavePlayer_2,WavePlayer_3], properties {weights [1,1,1]}}
 {id=WavePlayer_1, properties {soundWaveAssetPtr=&/SprayPaint_Sequence_Preview_01}}
 {id=WavePlayer_2, properties {soundWaveAssetPtr=&/SprayPaint_Sequence_Preview_02}}
 {id=WavePlayer_3, properties {soundWaveAssetPtr=&/SprayPaint_Sequence_Preview_03}}
]
properties {
 attenuationSettings=&Fort_Emotes_Athena_Att
 bOverrideConcurrency=true
 maxDistance=1600
 soundClassObject=&Fort_Music_Emote
 volumeMultiplier=0.2
 concurrencyOverrides {maxCount=1, resolutionRule=StopOldest}
}
